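# Translucent UI material for the Cluster Rocket Launcher that builds a masked, tintable dots pattern.
# The red channel of a grayscale dots texture is scaled by a Blur parameter, saturated, masked by an MF_UI_Box fill, then scaled by Translucency.
# The Circle Mask parameter drives the box size despite its name, while an MF_UI_Circle call sits unconnected.
# A near-white Color parameter provides the tint.
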Material /ClusterLauncher/UI/Materials/M_UI_ClusterRocketLauncher_DotsPattern
graph {
 properties {bAutomaticallySetUsageInEditor=false, blendMode=BLEND_Translucent, materialDomain=MD_UI, usedShadingModels=Default Lit}
 nodes [
  MaterialFunctionCall() {materialFunction=&MF_UI_Circle}
  n1 = Multiply(
   A=TextureSample() {samplerType=SAMPLERTYPE_Grayscale, texture=/ClusterLauncher/UI/Textures/T_UI_ClusterRocketLauncher_Dots}.R
   B=ScalarParameter() {defaultValue=1, parameterName=Blur}
  )
  n2 = ScalarParameter() {defaultValue=1, parameterName=Circle Mask}
  Multiply(
   A=ScalarParameter() {defaultValue=0.3, parameterName=Translucency}
   B=Multiply(A=Saturate(None=@n1), B=MaterialFunctionCall(Size=@n2) {materialFunction=&MF_UI_Box}.Fill)
  )
  VectorParameter() {parameterName=Color, defaultValue=<rgba 0.9911 0.9911 0.9911 1>}
 ]
}
