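# UI material that draws the frame of the Cluster Rocket Launcher reticle as a procedural SDF outline.
# A static switch picks a four-sided polygon or circle outline, which two perpendicular box SDFs cut into segments.
# Shared reticle functions build its UVs from scale and a positional spread offset.
# Anti-aliasing adds a faint drop shadow, while parameters blend hip-fire and ADS color and opacity.
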Material /ClusterLauncher/UI/Materials/M_UI_ClusterRocketLauncher_Frame
graph {
 properties {bAutomaticallySetUsageInEditor=false, blendMode=BLEND_AlphaComposite, materialDomain=MD_UI, usedShadingModels=Default Lit}
 nodes [
  Constant() {r=0.05}
  n1 = Constant()
  n2 = Constant()
  n3 = NamedRerouteUsage()
  n4 = ScalarParameter() {defaultValue=0.8, group=Shape Operators, parameterName=ShapeMaskPercent, sliderMax=1}
  n5 = MaterialFunctionCall(
   UVs=MaterialFunctionCall(
    UVs=MaterialFunctionCall() {materialFunction=&GetUserInterfaceUV}.Normalized UV
    "Rotation Angle (0-1)"=Add(A=@n1, B=@n2)
   ) {materialFunction=&CustomRotator}.Rotated Values
   Size=AppendVector(A=Add(A=@n3, B=@n4), B=Constant() {r=1})
  ) {materialFunction=&MF_UI_Box}
  n6 = Constant()
  n7 = Constant() {r=0.25}
  n8 = NamedRerouteUsage()
  n9 = ScalarParameter() {defaultValue=0.8, parameterName=ShapeMaskPercentHoriz}
  n10 = MaterialFunctionCall(
   UVs=MaterialFunctionCall(
    UVs=MaterialFunctionCall() {materialFunction=&GetUserInterfaceUV}.Normalized UV
    "Rotation Angle (0-1)"=Add(A=@n6, B=@n7)
   ) {materialFunction=&CustomRotator}.Rotated Values
   Size=AppendVector(A=Add(A=@n8, B=@n9), B=Constant() {r=1})
  ) {materialFunction=&MF_UI_Box}
  n11 = NamedRerouteUsage()
  n12 = MaterialFunctionCall(UVs=@n11, "Rotation Angle (0-1)"=Constant()) {materialFunction=&CustomRotator}
  n13 = ScalarParameter() {defaultValue=1, group=Shape Operators, parameterName=ShapeSize}
  n14 = Constant() {r=4}
  n15 = ScalarParameter() {defaultValue=0.02, parameterName=Frame Thickness}
  n16 = Constant() {r=-1}
  n17 = Max(
   A=Multiply(A=Multiply(A=@n5.SDF, B=@n10.SDF)) {constB=-10}
   B=StaticSwitch(
    True=MaterialFunctionCall!(UVs=@n12.Rotated Values, Size=@n13, Sides=@n14, InnerAngle, OutlineThickness=@n15, OutlineAlignment=@n16, GlowMin, GlowMax) {materialFunction=/FortUILibrary/MaterialFunctions/Shapes/MF_UI_RegularPolygon}.SDFOutline
    False=MaterialFunctionCall(UVs=@n11, Size=@n13, OutlineThickness=@n15, OutlineAlignment=@n16) {materialFunction=&MF_UI_Circle}.SDFOutline
    Value=StaticBoolParameter() {defaultValue=true, parameterName=Poly/Circle}
   )
  )
  n18 = MaterialFunctionCall(DistanceField=@n17) {materialFunction=&MF_UI_AntiAliasing}
  n19 = LinearInterpolate(
   A=@n18.Inner
   B=MaterialFunctionCall(DistanceField=@n17, Threshold=Constant() {desc=Controls the thickness of the Drop Shadow, r=0.005}) {materialFunction=&MF_UI_AntiAliasing}.Inner
   Alpha=Constant() {desc=Controls the opacity of the drop shadow., r=0.2}
  ) {constAlpha=0.4}
  n20 = NamedRerouteUsage()
  n21 = Multiply(
   A=@n19
   B=LinearInterpolate(
    A=ScalarParameter() {defaultValue=1, desc=Controls Overall Opacity, group=Reticle Operators, parameterName=HipFireOpacity, sliderMax=1}
    B=ScalarParameter() {defaultValue=1, desc=Controls Overall Opacity, group=Reticle Operators, parameterName=ADSOpacity, sliderMax=1}
    Alpha=ScalarParameter() {defaultValue=1, desc=Controls Overall Opacity, group=Reticle Operators, parameterName=IsADS, sliderMax=1}
   )
  )
  LinearInterpolate(
   A=Saturate(None=Subtract(A=@n19, B=@n20))
   B=@n21
   Alpha=ScalarParameter() {defaultValue=1, desc=Controls Overall Opacity, group=Reticle Operators, parameterName=MasterOpacity, sliderMax=1}
  )
  Multiply(
   A=@n21
   B=ScalarParameter() {defaultValue=1, desc=Controls Overall Opacity, group=Reticle Operators, parameterName=MasterOpacity, sliderMax=1}
  )
  n22 = VectorParameter() {parameterName=Default Color, defaultValue=<rgba 1 1 1 1>}
  n23 = VectorParameter() {parameterName=ADS Color, defaultValue=<rgba 1 0.1922 0 1>}
  n24 = ScalarParameter() {defaultValue=1, parameterName=ADSColorAnim}
  n25 = ScalarParameter() {defaultValue=1, parameterName=HipFireTransparency}
  n26 = ScalarParameter() {defaultValue=1, parameterName=Transparency}
  n27 = ScalarParameter() {defaultValue=1, desc=Controls Overall Opacity, group=Reticle Operators, parameterName=IsADS, sliderMax=1}
  Multiply(A=Multiply(A=LinearInterpolate(A=@n22.RGB, B=@n23.RGB, Alpha=@n24), B=LinearInterpolate(A=@n25, B=@n26, Alpha=@n27)), B=@n18.Inner)
  n28 = NamedRerouteUsage()
  NamedRerouteDeclaration(None=MaterialFunctionCall!(
    Reticle UVs
    Reticle Sides=Constant() {r=4}
    Reticle Scale=ScalarParameter() {defaultValue=1, parameterName=Scale}
    Reticle Rotation=Constant() {r=0.25}
    Prong Scale=ScalarParameter() {defaultValue=2, group=Prong Operators, parameterName=ProngScale, sliderMax=2}
    Prong Rotation=Constant() {r=0.5}
    Spread Offset=Multiply(A=@n28) {constB=-1}
    Positional Offset
    "Spread Input Range [Min,Max]"
    "Spread Remapped Range [Min,Max]"
   ) {materialFunction=/FortUILibrary/Systems/Reticles/MaterialFunctions/MF_UI_Reticle_UVs}.UVs) {name=Reticle Uvs, nodeColor=<rgba 1 0.7436 0 1>}
  n29 = Constant3Vector() {constant=<rgba 0 1 0 0>}
  n30 = ScalarParameter() {
   defaultValue=0.4
   desc=Variable to Position Reticle Prongs
   group=Reticle Operators
   parameterName=PositionalOffset
   sliderMax=1
   sliderMin=-1
  }
  NamedRerouteDeclaration(None=MaterialFunctionCall!(
    Positional Spread Offset=Multiply(A=@n29.G, B=@n30)
    Accuracy Spread Offset=Constant() {desc=Variable for Accuracy Spread Data}
    Min Value
    "Input Range [Min,Max]"=@n29
    "Remapped Range [Min,Max]"=Constant3Vector() {constant=<rgba -0.5 0 0 0>}
   ) {materialFunction=/FortUILibrary/Systems/Reticles/MaterialFunctions/MF_UI_Reticle_Remap_Spread}.Result) {name=Spread/Bloom, nodeColor=<rgba 1 0 0.1172 1>}
  n31 = Constant() {r=-0.002}
  NamedRerouteDeclaration(None=Multiply(A=MaterialFunctionCall(DistanceField=@n17, Threshold=@n31) {materialFunction=&MF_UI_AntiAliasing}.Inner)) {name=InnerGlow, nodeColor=<rgba 0.976 0.0308 0 1>}
 ]
}
