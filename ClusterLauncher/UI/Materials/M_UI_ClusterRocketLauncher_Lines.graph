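# M_UI_ClusterRocketLauncher_Lines is a translucent UI material for the line graphic of the Cluster Rocket Launcher interface.
# The red channel of a grayscale lines texture drives a lerp up to the Translucency parameter, producing a faint mask that is low by default.
# A near-white Color vector parameter supplies an adjustable tint, so material instances can change both color and line strength.
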Material /ClusterLauncher/UI/Materials/M_UI_ClusterRocketLauncher_Lines
graph {
 properties {bAutomaticallySetUsageInEditor=false, blendMode=BLEND_Translucent, materialDomain=MD_UI, usedShadingModels=Default Lit}
 nodes [
  LinearInterpolate(
   B=ScalarParameter() {defaultValue=0.08, parameterName=Translucency}
   Alpha=TextureSample() {samplerType=SAMPLERTYPE_Grayscale, texture=/ClusterLauncher/UI/Textures/T_UI_ClusterRocketLauncher_Lines}.R
  )
  VectorParameter() {parameterName=Color, defaultValue=<rgba 0.9911 0.9911 0.9911 1>}
 ]
}
