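# Translucent UI material that draws a horizontal reload bar driven by the ReloadProgress scalar parameter.
# A Step compares LinearGradient's U gradient to the progress, remapped to a slightly inset range, giving a hard-edged fill mask.
# MF_UI_Box supplies the Background fill and a slightly larger Stroke outline.
# Named reroutes carry these masks into lerp, multiply and add math that layers the bar.
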Material /ClusterLauncher/UI/Materials/M_UI_ReloadBar
graph {
 properties {bAutomaticallySetUsageInEditor=false, blendMode=BLEND_Translucent, materialDomain=MD_UI, usedShadingModels=Default Lit}
 nodes [
  n1 = NamedRerouteUsage()
  n2 = NamedRerouteUsage()
  n3 = OneMinus(None=Constant() {r=0.2})
  Add(
   A=LinearInterpolate(A=@n1, B=Multiply(A=@n1, B=@n2), Alpha=Multiply(A=@n1, B=@n3) {constB=0.1})
   B=Multiply(A=NamedRerouteUsage()) {constB=0.5}
  )
  LinearInterpolate(A=Constant() {r=1}, B=Constant(), Alpha=Subtract(A=NamedRerouteUsage(), B=NamedRerouteUsage()))
  n4 = ScalarParameter() {defaultValue=1, parameterName=ReloadProgress}
  NamedRerouteDeclaration(None=Step(
    Y=MaterialFunctionCall() {materialFunction=&LinearGradient}.UGradient
    X=LinearInterpolate(Alpha=@n4) {constA=0.04, constB=0.96}
   )) {name=ReloadProgress, nodeColor=<rgba 1 0 0.6328 1>}
  NamedRerouteDeclaration(None=MaterialFunctionCall(UVs=TextureCoordinate(), Size=Constant2Vector() {g=0.1, r=0.91}) {materialFunction=&MF_UI_Box}.Fill) {name=Background, nodeColor=<rgba 0 1 0.9766 1>}
  n5 = TextureCoordinate()
  n6 = Constant2Vector()
  NamedRerouteDeclaration(None=MaterialFunctionCall(UVs=Add(A=@n5, B=@n6), Size=Constant2Vector() {g=0.14, r=0.95}, OutlineThickness=Constant() {r=0.01}) {materialFunction=&MF_UI_Box}.Outline) {name=Stroke, nodeColor=<rgba 1 0.6562 0 1>}
 ]
}
