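# Reticle widget for the Cluster Launcher, showing ADS, target-in-range and reload states through material-driven UMG animations.
# Targeting changes play InADS forward or reversed; outside ADS, any actor type under the reticle within a curve-table max distance drives InRange.
# Both snap to their end state during the intro or before construction.
# Weapon reload events, bound on Construct, play a bar scaled to reload time, fading on finish or cancel.
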WidgetBlueprint /ClusterLauncher/UI/WBP_ClusterRocketLauncher_Reticle
animations [
 Reload {
  binding_count=1
  length_seconds=1
  object=&self:Reload
  sequence {bindings [Image_ReloadBar {
     id=#1
     possessed_class=Image
     tracks [
      {
       type=MovieSceneWidgetMaterialTrack
       sections [{
         type=MovieSceneParameterSection
         end_frame=20
         start_frame=0
         channels [ReloadProgress {type=MovieSceneScriptingFloatChannel, keys [[0,0],[20,1]]}]
        }]
      }
      {
       type=MovieSceneFloatTrack
       display_name=Render Opacity
       sections [{type=MovieSceneFloatSection, end_frame=0, start_frame=0, channels [None {type=MovieSceneScriptingFloatChannel, const [1,1]}]}]
      }
     ]
    }]}
 }
 ReloadFinished {
  binding_count=1
  length_seconds=0.3
  object=&self:ReloadFinished
  sequence {bindings [Image_ReloadBar {
     id=#2
     possessed_class=Image
     tracks [{
       type=MovieSceneFloatTrack
       display_name=Render Opacity
       sections [{type=MovieSceneFloatSection, end_frame=6, start_frame=0, channels [None {type=MovieSceneScriptingFloatChannel, keys [[0,1],[6,0]]}]}]
      }]
    }]}
 }
 Intro {
  binding_count=3
  length_seconds=0.7
  object=&self:Intro
  sequence {bindings [
    Image_Frame {
     id=#3
     possessed_class=Image
     tracks [{
       type=MovieSceneFloatTrack
       display_name=Render Opacity
       sections [{type=MovieSceneFloatSection, end_frame=14, start_frame=0, channels [None {type=MovieSceneScriptingFloatChannel, keys [[0,0],[14,1]]}]}]
      }]
    }
    Image_Dots {
     id=#4
     possessed_class=Image
     tracks [
      {
       type=MovieSceneWidgetMaterialTrack
       sections [{
         type=MovieSceneParameterSection
         end_frame=14
         start_frame=0
         channels [
          Blur {type=MovieSceneScriptingFloatChannel, keys [[0,1],[2,2],[4,1],[6,2],[8,1]]}
          Circle Mask {type=MovieSceneScriptingFloatChannel, keys [[6,1],[14,0]]}
          Translucency {type=MovieSceneScriptingFloatChannel, keys [[0,0],[2,0.5]]}
         ]
        }]
      }
      {
       type=MovieSceneColorTrack
       display_name=Color and Opacity
       sections [{
         type=MovieSceneColorSection
         end_frame=0
         start_frame=0
         channels [
          Color.R {type=MovieSceneScriptingFloatChannel, const [1,1]}
          Color.G {type=MovieSceneScriptingFloatChannel, const [1,1]}
          Color.B {type=MovieSceneScriptingFloatChannel, const [1,1]}
          Color.A {type=MovieSceneScriptingFloatChannel, const [1,1]}
         ]
        }]
      }
     ]
    }
    Image_DotMain {
     id=#5
     possessed_class=Image
     tracks [{
       type=MovieSceneFloatTrack
       display_name=Render Opacity
       sections [{type=MovieSceneFloatSection, end_frame=14, start_frame=0, channels [None {type=MovieSceneScriptingFloatChannel, keys [[13.222,0],[14,1]]}]}]
      }]
    }
   ]}
 }
 InADS {
  binding_count=3
  length_seconds=0.15
  object=&self:InADS
  sequence {bindings [
    Image_Frame {
     id=#6
     possessed_class=Image
     tracks [{
       type=MovieSceneWidgetMaterialTrack
       sections [{
         type=MovieSceneParameterSection
         end_frame=3
         start_frame=0
         channels [
          ProngScale {type=MovieSceneScriptingFloatChannel, keys [[0,4],[3,2]]}
          Frame Thickness {type=MovieSceneScriptingFloatChannel, keys [[0,0.007],[3,0.02]]}
          ShapeMaskPercent {type=MovieSceneScriptingFloatChannel, keys [[0,1.07],[3,0.8]]}
          ShapeMaskPercentHoriz {type=MovieSceneScriptingFloatChannel, keys [[0,1.05],[3,0.8]]}
          Scale {type=MovieSceneScriptingFloatChannel, const [1,2]}
          ADSColorAnim {type=MovieSceneScriptingFloatChannel, keys [[0,0],[1,1]]}
          ADSOpacity {type=MovieSceneScriptingFloatChannel, keys [[0,0.3],[3,1]]}
          Transparency {type=MovieSceneScriptingFloatChannel, const [1,1]}
          MasterOpacity {type=MovieSceneScriptingFloatChannel, keys [[0,1],[3,0.3]]}
         ]
        }]
      }]
    }
    Image_Lines {
     id=#7
     possessed_class=Image
     tracks [
      {
       type=MovieSceneWidgetMaterialTrack
       sections [{
         type=MovieSceneParameterSection
         end_frame=3
         start_frame=0
         channels [Translucency {type=MovieSceneScriptingFloatChannel, keys [[0,0],[3,0.15]]}]
        }]
      }
      {
       type=MovieScene2DTransformTrack
       sections [{
         type=MovieScene2DTransformSection
         end_frame=3
         start_frame=0
         channels [
          Translation.X {type=MovieSceneScriptingFloatChannel, const [0,1]}
          Translation.Y {type=MovieSceneScriptingFloatChannel, const [0,1]}
          Angle {type=MovieSceneScriptingFloatChannel, const [0,1]}
          Scale.X {type=MovieSceneScriptingFloatChannel, const [1,2]}
          Scale.Y {type=MovieSceneScriptingFloatChannel, keys [[0,0],[3,1]]}
          Shear.X {type=MovieSceneScriptingFloatChannel, const [0,1]}
          Shear.Y {type=MovieSceneScriptingFloatChannel, const [0,1]}
         ]
        }]
      }
     ]
    }
    Image_Dots {
     id=#8
     possessed_class=Image
     tracks [{
       type=MovieSceneWidgetMaterialTrack
       sections [{type=MovieSceneParameterSection, end_frame=3, start_frame=0, channels [Translucency {type=MovieSceneScriptingFloatChannel, const [0,2]}]}]
      }]
    }
   ]}
 }
 InRange {
  binding_count=1
  length_seconds=0.05
  object=&self:InRange
  sequence {bindings [Image_Frame {
     id=#9
     possessed_class=Image
     tracks [{
       type=MovieSceneWidgetMaterialTrack
       sections [{
         type=MovieSceneParameterSection
         end_frame=1
         start_frame=0
         channels [
          ProngScale {type=MovieSceneScriptingFloatChannel, keys [[0,4.25],[0.9997,4]]}
          Frame Thickness {type=MovieSceneScriptingFloatChannel, keys [[0,0.007],[1,0.01]]}
          ShapeMaskPercent {type=MovieSceneScriptingFloatChannel, const [1.07,1]}
          ShapeMaskPercentHoriz {type=MovieSceneScriptingFloatChannel, const [1.05,1]}
          HipFireOpacity {type=MovieSceneScriptingFloatChannel, keys [[0,0.4],[1,1]]}
          HipFireTransparency {type=MovieSceneScriptingFloatChannel, keys [[0,0.1],[1,1]]}
          Default ColorR {type=MovieSceneScriptingFloatChannel, const [1,1]}
          Default ColorG {type=MovieSceneScriptingFloatChannel, const [1,1]}
          Default ColorB {type=MovieSceneScriptingFloatChannel, const [1,1]}
          Default ColorA {type=MovieSceneScriptingFloatChannel, const [1,1]}
         ]
        }]
      }]
    }]}
 }
]
graphs [
 EventGraph {nodes [
   n1 = OnReloadStarted!() -> (OutputDelegate, then, ReloadTime:float=~, ReloadType:EFortWeaponReloadType Enum=ReloadWholeClip) {is=title K2Node_CustomEvent}
   > SetVisibility(self=Get Image_ReloadBar() -> (Image_ReloadBar:Image Object Reference), InVisibility=SelfHitTestInvisible)
   n2 = Get Reload() -> (Reload:Widget Animation Object Reference)
   > n3 = Play Animation with Finished event!(
    execute:Exec
    Widget:User Widget Object Reference=Self-Reference()
    InAnimation:Widget Animation Object Reference=@n2
    StartAtTime:float=0
    NumLoopsToPlay:int=1
    PlayMode:EUMGSequencePlayMode Enum
    PlaybackSpeed:float=float / float(A=GetEndTime(self=@n2), B=@n1.ReloadTime)
   ) -> (then:Exec, Result:Widget Animation Handle Structure, Finished:Exec) {is=title K2Node_PlayAnimation2}
   HideReloadBar(execute=@n3.Finished)
   n4 = OnADSAnimationFinished() {is=title K2Node_CustomEvent}
   > n5 = GetDynamicMaterial(self=Get Image_Frame() -> (Image_Frame:Image Object Reference))
   > SetScalarParameterValue(self=@n5, ParameterName=IsADS, Value="To Float (Boolean)"(InBool=Get IsInADS() -> (IsInADS:bool)))
   OnReloadCanceled() {is=title K2Node_CustomEvent}
   > PauseAnimation!(execute, self:User Widget Object Reference, InAnimation:Widget Animation Object Reference=Get Reload()) -> (then, ReturnValue:float)
   > HideReloadBar()
   HideReloadBar() {is=title K2Node_CustomEvent}
   > n6 = Play Animation with Finished event(
    Widget=Self-Reference()
    InAnimation=Get ReloadFinished() -> (ReloadFinished:Widget Animation Object Reference)
    StartAtTime=0
    NumLoopsToPlay=1
    PlaybackSpeed=1
   )
   SetVisibility(execute=@n6.Finished, self=Get Image_ReloadBar(), InVisibility=Hidden)
   n7 = Event OnTargetingChanged!() -> (OutputDelegate, then, bIsADS:bool) {is=title K2Node_Event}
   > SetIsInADS!(execute, self, NewIsInADS:bool=@n7.bIsADS) -> (then)
   > SetIsInRange!(execute, self, NewIsInRange:bool=IsInRangeState!(self) -> (InRange:bool)) -> (then)
   Event Construct()
   > Set IsConstructed(IsConstructed:bool=true) -> (Output_Get:bool)
   > SetMaxDistanceToActor!(
    execute
    self:Fort Actor Under Reticle VM Object Reference=Get FortActorUnderReticleVM() -> (FortActorUnderReticleVM:Fort Actor Under Reticle VM Object Reference)
    MaxDistance:float=GetValueAtLevel!(Input:"Scalable Float Structure (by ref)"=Get MaxDistance() -> (MaxDistance:Scalable Float Structure), Level:float=0) -> (ReturnValue:float)
   ) -> (then)
   > PlayAnimation(InAnimation=Get Intro())
   > SetVisibility(self=Get Image_ReloadBar(), InVisibility=Hidden)
   > n8 = Cast To B_TidalCrane_ClusterLauncher_Laser!(
    execute
    Object=GetEquippedWeapon!(self:Fort Weapon Reticle Extension Widget Base Object Reference) -> (ReturnValue:Fort Weapon Object Reference)
   ) -> (then, CastFailed, AsB Tidal Crane Cluster Launcher Laser:B Tidal Crane Cluster Launcher Laser Object Reference) {is=title K2Node_DynamicCast}
   > n9 = Set LauncherWeapon(LauncherWeapon:B Tidal Crane Cluster Launcher Laser Object Reference=@n8) -> (Output_Get:B Tidal Crane Cluster Launcher Laser Object Reference)
   n10 = Create Event()
   > Bind Event to On Reload Started(self:Fort Weapon Object Reference=@n9, Delegate=@n10) {is=title K2Node_AddDelegate}
   n11 = Create Event()
   > Bind Event to On Local Reload Cancelled(self:Fort Weapon Object Reference=[@n9, @n9], Delegate=@n11) {is=title K2Node_AddDelegate}
   > SetIsInADS(NewIsInADS=IsTargeting!(self:Fort Weapon Object Reference=@n9) -> (ReturnValue:bool))
   Event OnInitialized()
   > BindToAnimationFinished(Animation=Get InADS() -> (InADS:Widget Animation Object Reference), Delegate=@n4)
   Event Destruct()
   > Set IsConstructed(IsConstructed=false)
   > n12 = Get!(execute:Exec) -> (then:Exec, else:Exec, LauncherWeapon:B Tidal Crane Cluster Launcher Laser Object Reference)
   > Unbind Event from On Reload Started(self:Fort Weapon Object Reference=@n12, Delegate=@n10) {is=title K2Node_RemoveDelegate}
   > Unbind Event from On Local Reload Cancelled(self:Fort Weapon Object Reference=@n12, Delegate=@n11) {is=title K2Node_RemoveDelegate}
   > Set LauncherWeapon()
  ]}
 IsInRangeState {nodes [
   IsInRangeState() {is=title K2Node_FunctionEntry}
   n1 = Get FortActorUnderReticleVM()
   > Return Node!(
    execute
    InRange:bool=Select!(
     None:bool=false
     Enemy:bool=true
     Ally:bool=true
     World:bool=true
     Index:EActorUnderReticleType Enum=Get ActorUnderReticleType!(self:Fort Actor Under Reticle VM Object Reference=@n1) -> (ActorUnderReticleType:EActorUnderReticleType Enum=None)
    ) -> (ReturnValue:bool)
   )
  ]}
 OnUnderReticleTypeUpdated {nodes [
   OnUnderReticleTypeUpdated!() -> (then, Type:EActorUnderReticleType Enum=None) {is=title K2Node_FunctionEntry}
   > SetIsInRange(NewIsInRange=IsInRangeState() {is=title K2Node_CallFunction})
  ]}
 SetIsInADS {nodes [
   n1 = SetIsInADS!() -> (then, NewIsInADS:bool=~) {is=title K2Node_FunctionEntry}
   > n2 = GetDynamicMaterial(self=Get Image_Frame())
   > n3 = Get Scalar Parameter Value!(execute, self:Material Instance Dynamic Object Reference=@n2, ParameterName:Name=ADSColorAnim) -> (then, ReturnValue:float) {is=title K2Node_CallFunction}
   n4 = "Nearly Equal (Float)"(A=@n3, B="To Float (Boolean)"(InBool=Get IsInADS()), ErrorTolerance=0.001)
   n5 = Get IsInADS()
   > Branch(Condition=OR Boolean(A=NOT Boolean(A=@n4), B="Not Equal (Boolean)"(A=@n1, B=@n5) -> (ReturnValue=~)))
   > Set IsInADS(IsInADS:bool=Get NewIsInADS() -> (NewIsInADS:bool)) -> (Output_Get:bool)
   n6 = Get InADS()
   > PlayAnimation(
    InAnimation=@n6
    StartAtTime=Select!(Option 0:float=GetEndTime(self=@n6), Option 1:float=0, Index:bool=ShouldPlayAnimation!(self) -> (ShouldPlay:bool)) -> (ReturnValue:float)
    PlayMode=Select(Option 0:EUMGSequencePlayMode Enum=Reverse, Option 1:EUMGSequencePlayMode Enum, Index=Get IsInADS()) -> (ReturnValue:EUMGSequencePlayMode Enum)
   )
   > Branch(Condition=Get IsInADS())
   > n7 = GetDynamicMaterial(self=Get Image_Frame())
   > SetScalarParameterValue(self=@n7, ParameterName=IsADS, Value=1)
  ]}
 SetIsInRange {nodes [
   n1 = SetIsInRange!() -> (then, NewIsInRange:bool=~) {is=title K2Node_FunctionEntry}
   > n2 = Branch(Condition=Get IsInADS())
   Delta Gate Wildcard!(execute:Exec=@n2.else, Variable:"Boolean (by ref)"=Get IsInRange() -> (IsInRange:bool), New Value:bool=@n1) -> (Success:Exec, Updated Variable:"Boolean (by ref)", Previous Variable:bool=~) {is=title K2Node_MacroInstance}
   > n3 = Branch(Condition=ShouldPlayAnimation())
   > n4 = Branch(Condition=Get IsInRange())
   n5 = Get InRange() -> (InRange:Widget Animation Object Reference)
   > PlayAnimationForward(InAnimation=@n5)
   PlayAnimationReverse(execute=@n4.else, InAnimation=@n5)
   n6 = Get InRange()
   PlayAnimation(execute=@n3.else, InAnimation=@n6, StartAtTime=GetEndTime(self=@n6), PlayMode=Select(Option 0=Reverse, Index=Get IsInRange()))
  ]}
 ShouldPlayAnimation {nodes [
   ShouldPlayAnimation() {is=title K2Node_FunctionEntry}
   n1 = IsAnimationPlaying(InAnimation=Get Intro())
   > Return Node!(execute, ShouldPlay:bool=AND Boolean(A=Get IsConstructed() -> (IsConstructed:bool), B=NOT Boolean(A=@n1)))
  ]}
]
variables [
 IsConstructed {type=boolean}
 IsInADS {type=boolean}
 IsInRange {type=boolean}
 LauncherWeapon {type=B_TidalCrane_ClusterLauncher_Laser_C, default=null}
 MaxDistance {
  type=ScalableFloat
  default {
   registryType=None
   value=1
   curve {curveTable=/ClusterLauncher/DataTables/ClusterLauncherGameData, rowName=Default.ClusterLauncher.UI.InRange}
  }
 }
]
widgets {
 info {
  inheritedWidgetCount=1
  parentClass=/Script/FortniteUI.FortWeaponReticleExtensionWidgetBase
  rootWidgetClass=&GridPanel
  widgetCount=7
 }
 widgets [
  GridPanel_26 {widgetClassPath=&GridPanel, widget=GridPanel_26}
  Image_Dots {
   widgetClassPath=&Image
   parent=GridPanel_26
   widget=Image_Dots
   properties {
    visibility=SelfHitTestInvisible
    brush {
     bIsDynamicallyLoaded=false
     drawAs=Image
     imageType=FullColor
     mirroring=NoMirror
     resourceName=None
     resourceObject=/ClusterLauncher/UI/Materials/M_UI_ClusterRocketLauncher_DotsPattern
     tiling=NoTile
     imageSize=<32 32>
     margin=<ltrb 0 0 0 0>
     outlineSettings {
      bUseBrushTransparency=false
      roundingType=HalfHeightRadius
      width=0
      color {colorUseRule=UseColor_Specified, specifiedColor=<rgba 0 0 0 0>}
      cornerRadii=<0 0 0 0>
     }
     tintColor {colorUseRule=UseColor_Specified, specifiedColor=<rgba 1 1 1 1>}
     uVRegion {bIsValid=false, max=<0 0>, min=<0 0>}
    }
   }
   slot {class=GridSlot}
  }
  Image_Lines {
   widgetClassPath=&Image
   parent=GridPanel_26
   widget=Image_Lines
   properties {
    visibility=SelfHitTestInvisible
    brush {
     bIsDynamicallyLoaded=false
     drawAs=Image
     imageType=FullColor
     mirroring=NoMirror
     resourceName=None
     resourceObject=/ClusterLauncher/UI/Materials/M_UI_ClusterRocketLauncher_Lines
     tiling=NoTile
     imageSize=<32 32>
     margin=<ltrb 0 0 0 0>
     outlineSettings {
      bUseBrushTransparency=false
      roundingType=HalfHeightRadius
      width=0
      color {colorUseRule=UseColor_Specified, specifiedColor=<rgba 0 0 0 0>}
      cornerRadii=<0 0 0 0>
     }
     tintColor {colorUseRule=UseColor_Specified, specifiedColor=<rgba 1 1 1 1>}
     uVRegion {bIsValid=false, max=<0 0>, min=<0 0>}
    }
   }
   slot {class=GridSlot}
  }
  Image_Frame {
   widgetClassPath=&Image
   bIsVariable=true
   parent=GridPanel_26
   widget=Image_Frame
   properties {
    visibility=SelfHitTestInvisible
    brush {
     bIsDynamicallyLoaded=false
     drawAs=Image
     imageType=FullColor
     mirroring=NoMirror
     resourceName=None
     resourceObject=/ClusterLauncher/UI/Materials/M_UI_ClusterRocketLauncher_Frame
     tiling=NoTile
     imageSize=<210 172>
     margin=<ltrb 0 0 0 0>
     outlineSettings {
      bUseBrushTransparency=false
      roundingType=HalfHeightRadius
      width=0
      color {colorUseRule=UseColor_Specified, specifiedColor=<rgba 0 0 0 0>}
      cornerRadii=<0 0 0 0>
     }
     tintColor {colorUseRule=UseColor_Specified, specifiedColor=<rgba 1 1 1 1>}
     uVRegion {bIsValid=false, max=<0 0>, min=<0 0>}
    }
   }
   slot {class=GridSlot, properties {horizontalAlignment=HAlign_Center, verticalAlignment=VAlign_Center}}
  }
  Image_DotMain {
   widgetClassPath=&Image
   parent=GridPanel_26
   widget=Image_DotMain
   properties {
    visibility=SelfHitTestInvisible
    brush {
     bIsDynamicallyLoaded=false
     drawAs=Image
     imageType=FullColor
     mirroring=NoMirror
     resourceName=None
     resourceObject=/ClusterLauncher/UI/Materials/M_Reticle_Center
     tiling=NoTile
     imageSize=<15 15>
     margin=<ltrb 0 0 0 0>
     outlineSettings {
      bUseBrushTransparency=false
      roundingType=HalfHeightRadius
      width=0
      color {colorUseRule=UseColor_Specified, specifiedColor=<rgba 0 0 0 0>}
      cornerRadii=<0 0 0 0>
     }
     tintColor {colorUseRule=UseColor_Specified, specifiedColor=<rgba 1 1 1 1>}
     uVRegion {bIsValid=false, max=<0 0>, min=<0 0>}
    }
   }
   slot {class=GridSlot, properties {horizontalAlignment=HAlign_Center, verticalAlignment=VAlign_Center}}
  }
  Image_ReloadBar {
   widgetClassPath=&Image
   bIsVariable=true
   parent=GridPanel_26
   widget=Image_ReloadBar
   properties {
    visibility=SelfHitTestInvisible
    brush {
     bIsDynamicallyLoaded=false
     drawAs=Image
     imageType=FullColor
     mirroring=NoMirror
     resourceName=None
     resourceObject=/ClusterLauncher/UI/Materials/M_UI_ReloadBar
     tiling=NoTile
     imageSize=<150 150>
     margin=<ltrb 0 0 0 0>
     outlineSettings {
      bUseBrushTransparency=false
      roundingType=HalfHeightRadius
      width=0
      color {colorUseRule=UseColor_Specified, specifiedColor=<rgba 0 0 0 0>}
      cornerRadii=<0 0 0 0>
     }
     tintColor {colorUseRule=UseColor_Specified, specifiedColor=<rgba 1 1 1 1>}
     uVRegion {bIsValid=false, max=<0 0>, min=<0 0>}
    }
    renderTransform {angle=0, scale=<0.5 0.5>, shear=<0 0>, translation=<0 65>}
   }
   slot {class=GridSlot, properties {horizontalAlignment=HAlign_Center, verticalAlignment=VAlign_Center}}
  }
  TutorialHighlight {widgetClassPath=&TutorialHighlight}
 ]
}
